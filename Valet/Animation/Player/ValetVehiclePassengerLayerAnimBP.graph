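# Animation layer Blueprint that poses Fortnite players riding as passengers in Valet vehicles, including leaning out for drive-bys.
# On init it flags sports car or basic truck from the vehicle's anim instance to choose clips.
# Full-body layer blends seated/DBNO idle, action and turret poses, and a lean-driven drive-by state machine with slope aim offsets.
# While leaning, lower-body layer turns between four cardinal aim poses, with seat-side variants and aim offsets.
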AnimBlueprint /Valet/Animation/Player/ValetVehiclePassengerLayerAnimBP
target_skeleton=&Fortnite_M_Avg_Player_Skeleton
graphs [
 AnimGraph {nodes [Output Pose() {layerGroup=None}]}
 EventGraph {nodes [
   Event BlueprintInitializeAnimation()
   n1 = TryGetPawnOwner()
   > Is Valid(InputObject=@n1)
   > n2 = Cast To FortPlayerAnimInstance(Object=GetAnimInstance(self=GetComponentByClass!(self:Actor Object Reference=@n1, ComponentClass:Actor Component Class Reference=/Script/Engine.SkeletalMeshComponent) -> (ReturnValue:Skeletal Mesh Component Object Reference)))
   n3 = Get FortPlayerPawn!(self:Fort Player Anim Instance Object Reference=@n2) -> (FortPlayerPawn:Fort Player Pawn Object Reference)
   > Set Fort Player Pawn(Fort Player Pawn:Fort Player Pawn Object Reference=@n3) -> (Output_Get:Fort Player Pawn Object Reference)
   n4 = GetVehicleActor!(self:Fort Player Pawn Object Reference=@n3) -> (ReturnValue:Actor Object Reference)
   > Is Valid(InputObject=@n4)
   > n5 = Cast To FortDagwoodVehicle!(execute, Object=@n4) -> (then, CastFailed, AsFort Dagwood Vehicle:Fort Dagwood Vehicle Object Reference) {is=title K2Node_DynamicCast}
   > Set DagwoodVehicleInstance(DagwoodVehicleInstance:Fort Dagwood Vehicle Soft Object Reference=Make Soft Reference!(Input:Fort Dagwood Vehicle Object Reference=@n5) -> (Output:Fort Dagwood Vehicle Soft Object Reference)) -> (Output_Get:Fort Dagwood Vehicle Soft Object Reference)
   > n6 = Cast To FortVehicleAnimInstance!(
    execute
    Object=GetAnimInstance(self=Get SkeletalMesh!(self:Fort Athena Vehicle Object Reference=@n5) -> (SkeletalMesh:Skeletal Mesh Component Budgeted Object Reference))
   ) -> (then, CastFailed, AsFort Vehicle Anim Instance:Fort Vehicle Anim Instance Object Reference) {is=title K2Node_DynamicCast}
   n7 = Get VehicleType!(self:Fort Vehicle Anim Instance Object Reference=@n6) -> (VehicleType:EFortValetVehicleType Enum=Default)
   > n8 = Branch(Condition="Equal (Enum)"(A:EFortValetVehicleType Enum=@n7, B:EFortValetVehicleType Enum=Sport))
   > n9 = Set bIsSportsCar(bIsSportsCar:bool=true) -> (Output_Get:bool)
   n10 = Branch(execute=[@n11.then, @n9.then], Condition="Equal (Enum)"(A=@n7, B=BasicTruck))
   > Set bIsBasicTruck(bIsBasicTruck:bool=true) -> (Output_Get:bool)
   Set bIsBasicTruck(execute=@n10.else, bIsBasicTruck=false)
   n11 = Set bIsSportsCar(execute=@n8.else, bIsSportsCar=false)
  ]}
 VehicleFinalPoseOverride {nodes [
   Input Pose() {name=InFinalPose}
   Output Pose() {layerGroup=VehicleOverride}
  ]}
 VehicleFullBodyOverride {nodes [
   n1 = Layered blend per bone(BasePose=Use cached pose 'PassengerIdlePose'(), BlendPoses_0=Use cached pose 'UpperAndLowerPose'()) {
    bMeshSpaceRotationBlend=true
    layerSetup [{branchFilters [
       {boneName=Root}
       {blendDepth=-1, boneName=pelvis}
       {boneName=Spine_01}
       {blendDepth=-1, boneName=thigh_l}
       {blendDepth=-1, boneName=thigh_r}
       {blendDepth=-1, boneName=ik_foot_root}
      ]}]
   }
   n2 = Rotate Root Bone!(BasePose:Pose Link Structure=@n1, Pitch:float=0, Yaw:float=Get NegativeTurnYaw() -> (NegativeTurnYaw:float), MeshToComponent:Rotator="0,0,0") -> (Pose:Pose Link Structure) {is=title AnimGraphNode_RotateRootBone}
   AnimGraphNode_StateMachine_0 = DriveByTransition() {is=title AnimGraphNode_StateMachine}
   n3 = Get bIsInAction() -> (bIsInAction:bool)
   n4 = Layered blend per bone(BasePose=Use cached pose 'PassengerIdlePose'(), BlendPoses_0=Use cached pose 'UpperAndLowerPose'()) {
    bMeshSpaceRotationBlend=true
    layerSetup [{branchFilters [
       {boneName=Root}
       {blendDepth=-1, boneName=pelvis}
       {boneName=Spine_01}
       {blendDepth=-1, boneName=thigh_l}
       {blendDepth=-1, boneName=thigh_r}
       {blendDepth=-1, boneName=ik_foot_root}
      ]}]
   }
   n5 = Get NegativeTurnYaw()
   Output Pose(Result=Blend Poses by bool(
     BlendPose_0=Rotate Root Bone(BasePose=@n4, Pitch=0, Yaw=@n5, MeshToComponent="0,0,0")
     BlendPose_1=Blend Poses by bool(BlendPose_0=@n2, BlendPose_1=@AnimGraphNode_StateMachine_0, BlendTime_1=0.2, bActiveValue=@n3) {blendType=Linear, childUpateMode=ResetChildOnActivate}
     bActiveValue=Get bIsControllingTurret() -> (bIsControllingTurret:bool)
    ) {blendType=Linear}) {layerGroup=VehicleOverride}
   Save cached pose 'UpperAndLowerPose'(Pose=Input Pose() {name=InPoseFullBody}) {cacheName=UpperAndLowerPose}
   n6 = Apply Additive!(
    Base=Sequence Player 'ValetVehicle_NewSports_Passenger_Idle'() {sequence=/Valet/Animation/Player/SportsNew/ValetVehicle_NewSports_Passenger_Idle}
    Additive="Sequence Player (additive) 'ValetVehicle_NewSports_Passenger_Idle_CMF'"() {sequence=/Valet/Animation/Player/SportsNew/ValetVehicle_NewSports_Passenger_Idle_CMF}
    bAlphaBoolEnabled:bool=Get bIsFemale() -> (bIsFemale:bool)
   ) -> (Pose) {alphaInputType=Bool, alphaBoolBlend {blendOption=Linear}}
   n7 = Apply Additive(
    Base=Sequence Player 'ValetVehicle_Basic_Passenger_Idle'() {sequence=/Valet/Animation/Player/Basic/ValetVehicle_Basic_Passenger_Idle}
    Additive="Sequence Player (additive) 'ValetVehicle_Basic_Passenger_Idle_CMF'"() {sequence=/Valet/Animation/Player/Basic/ValetVehicle_Basic_Passenger_Idle_CMF}
    bAlphaBoolEnabled=Get bIsFemale()
   ) {alphaInputType=Bool, alphaBoolBlend {blendOption=Linear}}
   n8 = Get bIsSportsCar() -> (bIsSportsCar:bool)
   Save cached pose 'PassengerIdlePose'(Pose=Blend Poses by bool(
     BlendPose_0=Sequence Player 'TrendConcentrate_PassengerDBNO_AnimSET'() {sequence=/Game/Animation/Game/MainPlayer/Locomotion/Medium/Male/DBNO/TrendConcentrate_PassengerDBNO_AnimSET}
     BlendPose_1=Blend Poses by bool(BlendPose_0=@n6, BlendPose_1=@n7, bActiveValue=@n8) {blendType=Linear, childUpateMode=ResetChildOnActivate}
     bActiveValue=Get bIsDBNO()
    ) {blendType=Linear, childUpateMode=ResetChildOnActivate}) {cacheName=PassengerIdlePose}
  ]}
 DriveByTransition {
  entry=PassengerIdle
  path=VehicleFullBodyOverride.AnimGraphNode_StateMachine_0.DriveByTransition
  ids=id
  nodes [
   AnimStateNode_2 = PassengerIdle() {is=title AnimStateNode, bAlwaysResetOnEntry=true}
   AnimStateNode_3 = PassengerDriveBy() {is=title AnimStateNode}
   AnimStateNode_4 = PassengerDriveToIdle() {is=title AnimStateNode}
   AnimStateTransitionNode_2 = AnimStateTransitionNode() {crossfadeDuration=0}
   AnimStateTransitionNode_3 = AnimStateTransitionNode() {crossfadeDuration=0}
   AnimStateTransitionNode_4 = AnimStateTransitionNode() {bAutomaticRuleBasedOnSequencePlayerInState=true}
  ]
  links [
   {
    dst=PassengerDriveToIdle
    rule=VehicleFullBodyOverride.AnimGraphNode_StateMachine_0.DriveByTransition.AnimStateTransitionNode_3.Transition
    src=PassengerDriveBy
    via=AnimStateTransitionNode_3
   }
   {
    dst=PassengerIdle
    rule=VehicleFullBodyOverride.AnimGraphNode_StateMachine_0.DriveByTransition.AnimStateTransitionNode_4.Transition
    src=PassengerDriveToIdle
    via=AnimStateTransitionNode_4
   }
   {
    dst=PassengerDriveBy
    rule=VehicleFullBodyOverride.AnimGraphNode_StateMachine_0.DriveByTransition.AnimStateTransitionNode_2.Transition
    src=PassengerIdle
    via=AnimStateTransitionNode_2
   }
  ]
 }
 PassengerIdle {
  path=VehicleFullBodyOverride.AnimGraphNode_StateMachine_0.DriveByTransition.AnimStateNode_2.PassengerIdle
  nodes [Output Animation Pose(Result=Rotate Root Bone(BasePose=Use cached pose 'PassengerIdlePose'(), Pitch=0, Yaw=Get NegativeTurnYaw(), MeshToComponent="0,0,0"))]
 }
 PassengerDriveBy {
  path=VehicleFullBodyOverride.AnimGraphNode_StateMachine_0.DriveByTransition.AnimStateNode_3.PassengerDriveBy
  nodes [
   AnimGraphNode_StateMachine_0 = DriveBy() {is=title AnimGraphNode_StateMachine}
   n1 = "Transform (Modify) Bone - Bone: Root"!(ComponentPose=Local To Component(LocalPose=@AnimGraphNode_StateMachine_0), bAlphaBoolEnabled:bool=Get bIsFemale(), Translation="0,0,-8") -> (Pose) {alphaInputType=Bool, translationMode=BMM_Additive, alphaBoolBlend {blendOption=Linear}, boneToModify {boneName=Root}, translation {z=-8}}
   Output Animation Pose(Result=Component To Local(ComponentPose=@n1))
  ]
 }
 DriveBy {
  entry=DriveByStart
  path=VehicleFullBodyOverride.AnimGraphNode_StateMachine_0.DriveByTransition.AnimStateNode_3.PassengerDriveBy.AnimGraphNode_StateMachine_0.DriveBy
  ids=id
  nodes [
   AnimStateNode_0 = DriveByStart() {is=title AnimStateNode}
   AnimStateNode_1 = Default() {is=title AnimStateNode}
   AnimStateTransitionNode_1 = AnimStateTransitionNode() {bAutomaticRuleBasedOnSequencePlayerInState=true}
  ]
  links [{
    dst=Default
    rule=VehicleFullBodyOverride.AnimGraphNode_StateMachine_0.DriveByTransition.AnimStateNode_3.PassengerDriveBy.AnimGraphNode_StateMachine_0.DriveBy.AnimStateTransitionNode_1.AnimationTransitionGraph_0
    src=DriveByStart
    via=AnimStateTransitionNode_1
   }]
 }
 DriveByStart {
  path=VehicleFullBodyOverride.AnimGraphNode_StateMachine_0.DriveByTransition.AnimStateNode_3.PassengerDriveBy.AnimGraphNode_StateMachine_0.DriveBy.AnimStateNode_0.DriveByStart
  nodes [
   n1 = Get InTime() -> (InTime:float)
   n2 = Blend Poses by bool(
    BlendPose_0=Apply Additive(
     Base=Sequence Player 'ValetVehicle_BasicTruck_DriveByTransition'!(StartPosition:float=@n1) -> (Pose) {sequence=/Valet/Animation/Player/BasicTruck/ValetVehicle_BasicTruck_DriveByTransition}
     Additive="Sequence Player (additive) 'ValetVehicle_BasicTruck_DriveByTransition_CMF'"() {sequence=/Valet/Animation/Player/BasicTruck/ValetVehicle_BasicTruck_DriveByTransition_CMF}
     bAlphaBoolEnabled=Get bIsFemale()
    ) {alphaInputType=Bool, alphaBoolBlend {blendOption=Linear}}
    BlendPose_1=Apply Additive(
     Base=Sequence Player 'ValetVehicle_DriveByTransition'!(StartPosition:float=@n1) -> (Pose) {bLoopAnimation=false, sequence=/Valet/Animation/Player/Basic/ValetVehicle_DriveByTransition}
     Additive="Sequence Player (additive) 'ValetVehicle_DriveByTransition_CMF'"() {sequence=/Valet/Animation/Player/Basic/ValetVehicle_DriveByTransition_CMF}
     bAlphaBoolEnabled=Get bIsFemale()
    ) {alphaInputType=Bool, alphaBoolBlend {blendOption=Linear}}
    bActiveValue=Get bIsBasicTruck() -> (bIsBasicTruck:bool)
   ) {blendType=Linear}
   n3 = Sequence Player 'ValetVehicle_PL_DriveByTransition'!(StartPosition:float=@n1) -> (Pose) {bLoopAnimation=false, sequence=/Valet/Animation/Player/Basic/PassengerLeft/ValetVehicle_PL_DriveByTransition}
   n4 = Get bIsPassengerBackLeft() -> (bIsPassengerBackLeft:bool)
   Output Animation Pose(Result=Rotate Root Bone(
     BasePose=Blend Poses by bool(BlendPose_0=@n3, BlendPose_1=@n2, bActiveValue=@n4) {blendType=Linear}
     Pitch=0
     Yaw=Get NegativeTurnYaw()
     MeshToComponent="0,0,0"
    ))
   Get OutTime() -> (OutTime:float)
  ]
 }
 Default {
  path=VehicleFullBodyOverride.AnimGraphNode_StateMachine_0.DriveByTransition.AnimStateNode_3.PassengerDriveBy.AnimGraphNode_StateMachine_0.DriveBy.AnimStateNode_1.Default
  nodes [
   n1 = Use cached pose 'UpperAndLowerPose'()
   n2 = Component To Local(ComponentPose=Reset Root Transform!(ComponentPose:Component Space Pose Link Structure=Local To Component(LocalPose=@n1), Alpha:float=1) -> (Pose:Component Space Pose Link Structure) {is=title AnimGraphNode_ResetRoot, alphaBoolBlend {blendOption=Linear}})
   n3 = Get SlopePitchDegreeAngle() -> (SlopePitchDegreeAngle:float)
   Output Animation Pose(Result=AimOffset Player 'Valet_Passenger_Slope_RollAO'!(
     X:float=Get SlopeRollDegreeAngle() -> (SlopeRollDegreeAngle:float)
     BasePose:Pose Link Structure=AimOffset Player 'Valet_Passenger_Slope_PitchAO'!(X:float=@n3, BasePose:Pose Link Structure=@n2, Alpha:float=1) -> (Pose:Pose Link Structure) {blendSpace=/Valet/Animation/Player/Basic/Valet_Passenger_Slope_PitchAO, alphaBoolBlend {blendOption=Linear}}
     Alpha:float=1
    ) -> (Pose:Pose Link Structure) {blendSpace=/Valet/Animation/Player/Basic/Valet_Passenger_Slope_RollAO, alphaBoolBlend {blendOption=Linear}})
  ]
 }
 AnimationTransitionGraph_0 {
  path=VehicleFullBodyOverride.AnimGraphNode_StateMachine_0.DriveByTransition.AnimStateNode_3.PassengerDriveBy.AnimGraphNode_StateMachine_0.DriveBy.AnimStateTransitionNode_1.AnimationTransitionGraph_0
  nodes [Result()]
 }
 PassengerDriveToIdle {
  path=VehicleFullBodyOverride.AnimGraphNode_StateMachine_0.DriveByTransition.AnimStateNode_4.PassengerDriveToIdle
  nodes [
   n1 = Get OutTime()
   n2 = Sequence Player 'ValetVehicle_DriveByToIdle_CMM'!(PlayRate:float=1.2, StartPosition:float=@n1) -> (Pose) {bLoopAnimation=false, sequence=/Valet/Animation/Player/Basic/ValetVehicle_DriveByToIdle_CMM}
   n3 = Sequence Player 'ValetVehicle_PL_DriveByToIdle_CMM'!(PlayRate:float=1.2, StartPosition:float=@n1) -> (Pose) {bLoopAnimation=false, sequence=/Valet/Animation/Player/Basic/PassengerLeft/ValetVehicle_PL_DriveByToIdle_CMM}
   n4 = Get bIsPassengerBackLeft()
   n5 = "Transform (Modify) Bone - Bone: Root"(
    ComponentPose=Local To Component(LocalPose=Blend Poses by bool(BlendPose_0=@n3, BlendPose_1=@n2, bActiveValue=@n4) {blendType=Linear})
    bAlphaBoolEnabled=Get bIsFemale()
    Translation="0,0,-8"
   ) {alphaInputType=Bool, translationMode=BMM_Additive, alphaBoolBlend {blendOption=Linear}, boneToModify {boneName=Root}, translation {z=-8}}
   Output Animation Pose(Result=Rotate Root Bone(BasePose=Component To Local(ComponentPose=@n5), Pitch=0, Yaw=Get NegativeTurnYaw(), MeshToComponent="0,0,0"))
  ]
 }
 Transition {
  path=VehicleFullBodyOverride.AnimGraphNode_StateMachine_0.DriveByTransition.AnimStateTransitionNode_2.Transition
  nodes [
   Result(bCanEnterTransition=Get bIsLeaning() -> (bIsLeaning:bool))
   "Get Relevant Anim Time Remaining (PassengerIdle)"!() -> (ReturnValue:float) {is=title K2Node_AnimGetter}
  ]
 }
 Transition {
  path=VehicleFullBodyOverride.AnimGraphNode_StateMachine_0.DriveByTransition.AnimStateTransitionNode_3.Transition
  nodes [Result(bCanEnterTransition=NOT Boolean(A=Get bIsLeaning()))]
 }
 Transition {path=VehicleFullBodyOverride.AnimGraphNode_StateMachine_0.DriveByTransition.AnimStateTransitionNode_4.Transition, nodes [Result()]}
 VehicleLowerBodyOverride {nodes [
   Input Pose() {name=InPoseUpperBody}
   AnimGraphNode_StateMachine_0 = ValetPassengerLowerBodyTurn() {is=title AnimGraphNode_StateMachine}
   n1 = Get NegativeTurnYaw()
   Output Pose(Result=Blend Poses by bool(
     BlendPose_0=Rotate Root Bone(BasePose=@AnimGraphNode_StateMachine_0, Pitch=0, Yaw=@n1, MeshToComponent="0,0,0")
     BlendPose_1=Input Pose() {name=InPoseLowerBodyDefault}
     bActiveValue=Get bIsLeaning()
    ) {blendType=Linear}) {layerGroup=VehicleOverride}
   Sequence Player 'AncientWeapon_Crouch_NonTargeted_CMM'() {bLoopAnimation=false, sequence=/Game/Animation/Game/MainPlayerFN/Poses/AncientWeapon/AncientWeapon_Crouch_NonTargeted_CMM}
  ]}
 ValetPassengerLowerBodyTurn {
  entry=PassengerPose_N
  path=VehicleLowerBodyOverride.AnimGraphNode_StateMachine_0.ValetPassengerLowerBodyTurn
  ids=id
  nodes [
   AnimStateNode_10 = Transition_EtoS() {is=title AnimStateNode}
   AnimStateNode_11 = PassengerPose_S() {is=title AnimStateNode}
   AnimStateNode_12 = Transition_StoW() {is=title AnimStateNode}
   AnimStateNode_13 = PassengerPose_W() {is=title AnimStateNode}
   AnimStateNode_15 = Transition_WtoN() {is=title AnimStateNode}
   AnimStateNode_16 = PassengerPose_N() {is=title AnimStateNode}
   AnimStateNode_17 = Transition_EtoN() {is=title AnimStateNode}
   AnimStateNode_18 = Transition_StoE() {is=title AnimStateNode}
   AnimStateNode_19 = Transition_WtoS() {is=title AnimStateNode}
   AnimStateNode_20 = Transition_NtoW() {is=title AnimStateNode}
   AnimStateNode_8 = Transition_NtoE() {is=title AnimStateNode}
   AnimStateNode_9 = PassengerPose_E() {is=title AnimStateNode}
   AnimStateTransitionNode_0 = AnimStateTransitionNode()
   AnimStateTransitionNode_1 = AnimStateTransitionNode() {bAutomaticRuleBasedOnSequencePlayerInState=true}
   AnimStateTransitionNode_10 = AnimStateTransitionNode() {priorityOrder=2}
   AnimStateTransitionNode_11 = AnimStateTransitionNode() {priorityOrder=2}
   AnimStateTransitionNode_2 = AnimStateTransitionNode() {bAutomaticRuleBasedOnSequencePlayerInState=true}
   AnimStateTransitionNode_25 = AnimStateTransitionNode()
   AnimStateTransitionNode_26 = AnimStateTransitionNode() {bAutomaticRuleBasedOnSequencePlayerInState=true}
   AnimStateTransitionNode_27 = AnimStateTransitionNode()
   AnimStateTransitionNode_28 = AnimStateTransitionNode()
   AnimStateTransitionNode_29 = AnimStateTransitionNode()
   AnimStateTransitionNode_3 = AnimStateTransitionNode() {bAutomaticRuleBasedOnSequencePlayerInState=true}
   AnimStateTransitionNode_30 = AnimStateTransitionNode()
   AnimStateTransitionNode_31 = AnimStateTransitionNode()
   AnimStateTransitionNode_32 = AnimStateTransitionNode()
   AnimStateTransitionNode_35 = AnimStateTransitionNode()
   AnimStateTransitionNode_36 = AnimStateTransitionNode()
   AnimStateTransitionNode_37 = AnimStateTransitionNode()
   AnimStateTransitionNode_38 = AnimStateTransitionNode() {bAutomaticRuleBasedOnSequencePlayerInState=true}
   AnimStateTransitionNode_39 = AnimStateTransitionNode()
   AnimStateTransitionNode_4 = AnimStateTransitionNode()
   AnimStateTransitionNode_40 = AnimStateTransitionNode() {bAutomaticRuleBasedOnSequencePlayerInState=true}
   AnimStateTransitionNode_41 = AnimStateTransitionNode()
   AnimStateTransitionNode_42 = AnimStateTransitionNode() {bAutomaticRuleBasedOnSequencePlayerInState=true}
   AnimStateTransitionNode_5 = AnimStateTransitionNode()
   AnimStateTransitionNode_6 = AnimStateTransitionNode()
   AnimStateTransitionNode_7 = AnimStateTransitionNode() {bAutomaticRuleBasedOnSequencePlayerInState=true}
   AnimStateTransitionNode_8 = AnimStateTransitionNode() {priorityOrder=2}
   AnimStateTransitionNode_9 = AnimStateTransitionNode() {priorityOrder=2}
  ]
  links [
   {
    dst=PassengerPose_W
    rule=VehicleLowerBodyOverride.AnimGraphNode_StateMachine_0.ValetPassengerLowerBodyTurn.AnimStateTransitionNode_11.AnimationTransitionGraph_0
    src=PassengerPose_E
    via=AnimStateTransitionNode_11
   }
   {
    dst=Transition_EtoN
    rule=VehicleLowerBodyOverride.AnimGraphNode_StateMachine_0.ValetPassengerLowerBodyTurn.AnimStateTransitionNode_35.AnimationTransitionGraph_0
    src=PassengerPose_E
    via=AnimStateTransitionNode_35
   }
   {
    dst=Transition_EtoS
    rule=VehicleLowerBodyOverride.AnimGraphNode_StateMachine_0.ValetPassengerLowerBodyTurn.AnimStateTransitionNode_27.AnimationTransitionGraph_1
    src=PassengerPose_E
    via=AnimStateTransitionNode_27
   }
   {
    dst=PassengerPose_S
    rule=VehicleLowerBodyOverride.AnimGraphNode_StateMachine_0.ValetPassengerLowerBodyTurn.AnimStateTransitionNode_8.AnimationTransitionGraph_1
    src=PassengerPose_N
    via=AnimStateTransitionNode_8
   }
   {
    dst=Transition_NtoE
    rule=VehicleLowerBodyOverride.AnimGraphNode_StateMachine_0.ValetPassengerLowerBodyTurn.AnimStateTransitionNode_25.AnimationTransitionGraph_0
    src=PassengerPose_N
    via=AnimStateTransitionNode_25
   }
   {
    dst=Transition_NtoW
    rule=VehicleLowerBodyOverride.AnimGraphNode_StateMachine_0.ValetPassengerLowerBodyTurn.AnimStateTransitionNode_41.AnimationTransitionGraph_0
    src=PassengerPose_N
    via=AnimStateTransitionNode_41
   }
   {
    dst=PassengerPose_N
    rule=VehicleLowerBodyOverride.AnimGraphNode_StateMachine_0.ValetPassengerLowerBodyTurn.AnimStateTransitionNode_9.AnimationTransitionGraph_1
    src=PassengerPose_S
    via=AnimStateTransitionNode_9
   }
   {
    dst=Transition_StoE
    rule=VehicleLowerBodyOverride.AnimGraphNode_StateMachine_0.ValetPassengerLowerBodyTurn.AnimStateTransitionNode_37.AnimationTransitionGraph_0
    src=PassengerPose_S
    via=AnimStateTransitionNode_37
   }
   {
    dst=Transition_StoW
    rule=VehicleLowerBodyOverride.AnimGraphNode_StateMachine_0.ValetPassengerLowerBodyTurn.AnimStateTransitionNode_29.AnimationTransitionGraph_1
    src=PassengerPose_S
    via=AnimStateTransitionNode_29
   }
   {
    dst=PassengerPose_E
    rule=VehicleLowerBodyOverride.AnimGraphNode_StateMachine_0.ValetPassengerLowerBodyTurn.AnimStateTransitionNode_10.AnimationTransitionGraph_0
    src=PassengerPose_W
    via=AnimStateTransitionNode_10
   }
   {
    dst=Transition_WtoN
    rule=VehicleLowerBodyOverride.AnimGraphNode_StateMachine_0.ValetPassengerLowerBodyTurn.AnimStateTransitionNode_31.AnimationTransitionGraph_0
    src=PassengerPose_W
    via=AnimStateTransitionNode_31
   }
   {
    dst=Transition_WtoS
    rule=VehicleLowerBodyOverride.AnimGraphNode_StateMachine_0.ValetPassengerLowerBodyTurn.AnimStateTransitionNode_39.AnimationTransitionGraph_1
    src=PassengerPose_W
    via=AnimStateTransitionNode_39
   }
   {
    dst=PassengerPose_N
    rule=VehicleLowerBodyOverride.AnimGraphNode_StateMachine_0.ValetPassengerLowerBodyTurn.AnimStateTransitionNode_36.AnimationTransitionGraph_0
    src=Transition_EtoN
    via=AnimStateTransitionNode_36
   }
   {
    dst=PassengerPose_N
    rule=VehicleLowerBodyOverride.AnimGraphNode_StateMachine_0.ValetPassengerLowerBodyTurn.AnimStateTransitionNode_7.AnimationTransitionGraph_0
    src=Transition_EtoN
    via=AnimStateTransitionNode_7
   }
   {
    dst=PassengerPose_S
    rule=VehicleLowerBodyOverride.AnimGraphNode_StateMachine_0.ValetPassengerLowerBodyTurn.AnimStateTransitionNode_2.AnimationTransitionGraph_1
    src=Transition_EtoS
    via=AnimStateTransitionNode_2
   }
   {
    dst=PassengerPose_S
    rule=VehicleLowerBodyOverride.AnimGraphNode_StateMachine_0.ValetPassengerLowerBodyTurn.AnimStateTransitionNode_28.AnimationTransitionGraph_1
    src=Transition_EtoS
    via=AnimStateTransitionNode_28
   }
   {
    dst=PassengerPose_E
    rule=VehicleLowerBodyOverride.AnimGraphNode_StateMachine_0.ValetPassengerLowerBodyTurn.AnimStateTransitionNode_0.AnimationTransitionGraph_0
    src=Transition_NtoE
    via=AnimStateTransitionNode_0
   }
   {
    dst=PassengerPose_E
    rule=VehicleLowerBodyOverride.AnimGraphNode_StateMachine_0.ValetPassengerLowerBodyTurn.AnimStateTransitionNode_26.AnimationTransitionGraph_0
    src=Transition_NtoE
    via=AnimStateTransitionNode_26
   }
   {
    dst=PassengerPose_W
    rule=VehicleLowerBodyOverride.AnimGraphNode_StateMachine_0.ValetPassengerLowerBodyTurn.AnimStateTransitionNode_4.AnimationTransitionGraph_0
    src=Transition_NtoW
    via=AnimStateTransitionNode_4
   }
   {
    dst=PassengerPose_W
    rule=VehicleLowerBodyOverride.AnimGraphNode_StateMachine_0.ValetPassengerLowerBodyTurn.AnimStateTransitionNode_42.AnimationTransitionGraph_0
    src=Transition_NtoW
    via=AnimStateTransitionNode_42
   }
   {
    dst=PassengerPose_E
    rule=VehicleLowerBodyOverride.AnimGraphNode_StateMachine_0.ValetPassengerLowerBodyTurn.AnimStateTransitionNode_38.AnimationTransitionGraph_1
    src=Transition_StoE
    via=AnimStateTransitionNode_38
   }
   {
    dst=PassengerPose_E
    rule=VehicleLowerBodyOverride.AnimGraphNode_StateMachine_0.ValetPassengerLowerBodyTurn.AnimStateTransitionNode_6.AnimationTransitionGraph_1
    src=Transition_StoE
    via=AnimStateTransitionNode_6
   }
   {
    dst=PassengerPose_W
    rule=VehicleLowerBodyOverride.AnimGraphNode_StateMachine_0.ValetPassengerLowerBodyTurn.AnimStateTransitionNode_1.AnimationTransitionGraph_1
    src=Transition_StoW
    via=AnimStateTransitionNode_1
   }
   {
    dst=PassengerPose_W
    rule=VehicleLowerBodyOverride.AnimGraphNode_StateMachine_0.ValetPassengerLowerBodyTurn.AnimStateTransitionNode_30.AnimationTransitionGraph_1
    src=Transition_StoW
    via=AnimStateTransitionNode_30
   }
   {
    dst=PassengerPose_N
    rule=VehicleLowerBodyOverride.AnimGraphNode_StateMachine_0.ValetPassengerLowerBodyTurn.AnimStateTransitionNode_3.AnimationTransitionGraph_0
    src=Transition_WtoN
    via=AnimStateTransitionNode_3
   }
   {
    dst=PassengerPose_N
    rule=VehicleLowerBodyOverride.AnimGraphNode_StateMachine_0.ValetPassengerLowerBodyTurn.AnimStateTransitionNode_32.AnimationTransitionGraph_0
    src=Transition_WtoN
    via=AnimStateTransitionNode_32
   }
   {
    dst=PassengerPose_S
    rule=VehicleLowerBodyOverride.AnimGraphNode_StateMachine_0.ValetPassengerLowerBodyTurn.AnimStateTransitionNode_40.AnimationTransitionGraph_0
    src=Transition_WtoS
    via=AnimStateTransitionNode_40
   }
   {
    dst=PassengerPose_S
    rule=VehicleLowerBodyOverride.AnimGraphNode_StateMachine_0.ValetPassengerLowerBodyTurn.AnimStateTransitionNode_5.AnimationTransitionGraph_0
    src=Transition_WtoS
    via=AnimStateTransitionNode_5
   }
  ]
 }
 Transition_EtoS {
  path=VehicleLowerBodyOverride.AnimGraphNode_StateMachine_0.ValetPassengerLowerBodyTurn.AnimStateNode_10.Transition_EtoS
  nodes [Output Animation Pose(Result=Blend Poses by bool(
     BlendPose_0=Sequence Player 'ValetVehicle_PL_DriveByTurnEtoS'() {bLoopAnimation=false, sequence=/Valet/Animation/Player/Basic/PassengerLeft/ValetVehicle_PL_DriveByTurnEtoS}
     BlendPose_1=Sequence Player 'ValetVehicle_DriveByTurnEtoS'() {bLoopAnimation=false, sequence=/Valet/Animation/Player/Basic/ValetVehicle_DriveByTurnEtoS}
     bActiveValue=Get bIsPassengerBackLeft()
    ) {blendType=Linear})]
 }
 PassengerPose_S {
  path=VehicleLowerBodyOverride.AnimGraphNode_StateMachine_0.ValetPassengerLowerBodyTurn.AnimStateNode_11.PassengerPose_S
  nodes [
   n1 = Sequence Player 'ValetVehicle_PL_DriveBy_S_Pose'() {sequence=/Valet/Animation/Player/Basic/PassengerLeft/ValetVehicle_PL_DriveBy_S_Pose}
   n2 = Sequence Player 'ValetVehicle_Passenger_DriveBy_S_Pose'() {sequence=/Valet/Animation/Player/Basic/ValetVehicle_Passenger_DriveBy_S_Pose}
   n3 = Get bIsPassengerBackLeft()
   n4 = AimOffset Player 'ShoppingCart_Passenger_LowerBodyAO'!(
    X:float=Get AimBWDDeltaAngleDegrees() -> (AimBWDDeltaAngleDegrees:float)
    BasePose:Pose Link Structure=Rotate Root Bone!(BasePose=Blend Poses by bool(BlendPose_0=@n1, BlendPose_1=@n2, bActiveValue=@n3) {blendType=Linear}, Yaw=180, MeshToComponent="0,0,0") -> (Pose)
    Alpha:float=1
   ) -> (Pose:Pose Link Structure) {blendSpace=/Game/Animation/Game/MainPlayer/Vehicles/ShoppingCart/AO/ShoppingCart_Passenger_LowerBodyAO, alphaBoolBlend {blendOption=Linear}}
   Output Animation Pose(Result=Rotate Root Bone(BasePose=@n4, Yaw=-180, MeshToComponent="0,0,0"))
  ]
 }
 Transition_StoW {
  path=VehicleLowerBodyOverride.AnimGraphNode_StateMachine_0.ValetPassengerLowerBodyTurn.AnimStateNode_12.Transition_StoW
  nodes [Output Animation Pose(Result=Blend Poses by bool(
     BlendPose_0=Sequence Player 'ValetVehicle_PL_DriveByTurnStoW'() {bLoopAnimation=false, sequence=/Valet/Animation/Player/Basic/PassengerLeft/ValetVehicle_PL_DriveByTurnStoW}
     BlendPose_1=Sequence Player 'ValetVehicle_DriveByTurnStoW'() {bLoopAnimation=false, sequence=/Valet/Animation/Player/Basic/ValetVehicle_DriveByTurnStoW}
     bActiveValue=Get bIsPassengerBackLeft()
    ) {blendType=Linear})]
 }
 PassengerPose_W {
  path=VehicleLowerBodyOverride.AnimGraphNode_StateMachine_0.ValetPassengerLowerBodyTurn.AnimStateNode_13.PassengerPose_W
  nodes [
   n1 = Sequence Player 'ValetVehicle_PL_DriveBy_W_Pose'() {sequence=/Valet/Animation/Player/Basic/PassengerLeft/ValetVehicle_PL_DriveBy_W_Pose}
   n2 = Sequence Player 'ValetVehicle_Passenger_DriveBy_W_Pose'() {sequence=/Valet/Animation/Player/Basic/ValetVehicle_Passenger_DriveBy_W_Pose}
   n3 = Get bIsPassengerBackLeft()
   n4 = AimOffset Player 'ShoppingCart_Passenger_LowerBodyAO'(
    X=Get AimLFTDeltaAngleDegrees() -> (AimLFTDeltaAngleDegrees:float)
    BasePose=Rotate Root Bone(BasePose=Blend Poses by bool(BlendPose_0=@n1, BlendPose_1=@n2, bActiveValue=@n3) {blendType=Linear}, Yaw=90, MeshToComponent="0,0,0")
    Alpha=1
   ) {blendSpace=&ShoppingCart_Passenger_LowerBodyAO, alphaBoolBlend {blendOption=Linear}}
   Output Animation Pose(Result=Rotate Root Bone(BasePose=@n4, Yaw=-90, MeshToComponent="0,0,0"))
  ]
 }
 Transition_WtoN {
  path=VehicleLowerBodyOverride.AnimGraphNode_StateMachine_0.ValetPassengerLowerBodyTurn.AnimStateNode_15.Transition_WtoN
  nodes [Output Animation Pose(Result=Blend Poses by bool(
     BlendPose_0=Sequence Player 'ValetVehicle_PL_DriveByTurnWtoN'() {bLoopAnimation=false, sequence=/Valet/Animation/Player/Basic/PassengerLeft/ValetVehicle_PL_DriveByTurnWtoN}
     BlendPose_1=Sequence Player 'ValetVehicle_DriveByTurnWtoN'() {bLoopAnimation=false, sequence=/Valet/Animation/Player/Basic/ValetVehicle_DriveByTurnWtoN}
     bActiveValue=Get bIsPassengerBackLeft()
    ) {blendType=Linear})]
 }
 PassengerPose_N {
  path=VehicleLowerBodyOverride.AnimGraphNode_StateMachine_0.ValetPassengerLowerBodyTurn.AnimStateNode_16.PassengerPose_N
  nodes [
   n1 = AimOffset Player 'ShoppingCart_Passenger_LowerBodyAO'(
    X=Get AimFWDDeltaAngleDegrees() -> (AimFWDDeltaAngleDegrees:float)
    BasePose=Blend Poses by bool(
     BlendPose_0=Sequence Player 'ValetVehicle_PL_DriveBy_N_Pose'() {sequence=/Valet/Animation/Player/Basic/PassengerLeft/ValetVehicle_PL_DriveBy_N_Pose}
     BlendPose_1=Sequence Player 'ValetVehicle_Passenger_DriveBy_N_Pose'() {sequence=/Valet/Animation/Player/Basic/ValetVehicle_Passenger_DriveBy_N_Pose}
     bActiveValue=Get bIsPassengerBackLeft()
    ) {blendType=Linear}
    Alpha=1
   ) {blendSpace=&ShoppingCart_Passenger_LowerBodyAO, alphaBoolBlend {blendOption=Linear}}
   Output Animation Pose(Result=@n1)
  ]
 }
 Transition_EtoN {
  path=VehicleLowerBodyOverride.AnimGraphNode_StateMachine_0.ValetPassengerLowerBodyTurn.AnimStateNode_17.Transition_EtoN
  nodes [Output Animation Pose(Result=Blend Poses by bool(
     BlendPose_0=Sequence Player 'ValetVehicle_PL_DriveByTurnEtoN'() {bLoopAnimation=false, sequence=/Valet/Animation/Player/Basic/PassengerLeft/ValetVehicle_PL_DriveByTurnEtoN}
     BlendPose_1=Sequence Player 'ValetVehicle_DriveByTurnEtoN'() {bLoopAnimation=false, sequence=/Valet/Animation/Player/Basic/ValetVehicle_DriveByTurnEtoN}
     bActiveValue=Get bIsPassengerBackLeft()
    ) {blendType=Linear})]
 }
 Transition_StoE {
  path=VehicleLowerBodyOverride.AnimGraphNode_StateMachine_0.ValetPassengerLowerBodyTurn.AnimStateNode_18.Transition_StoE
  nodes [Output Animation Pose(Result=Blend Poses by bool(
     BlendPose_0=Sequence Player 'ValetVehicle_PL_DriveByTurnStoE'() {bLoopAnimation=false, sequence=/Valet/Animation/Player/Basic/PassengerLeft/ValetVehicle_PL_DriveByTurnStoE}
     BlendPose_1=Sequence Player 'ValetVehicle_DriveByTurnStoE'() {bLoopAnimation=false, sequence=/Valet/Animation/Player/Basic/ValetVehicle_DriveByTurnStoE}
     bActiveValue=Get bIsPassengerBackLeft()
    ) {blendType=Linear})]
 }
 Transition_WtoS {
  path=VehicleLowerBodyOverride.AnimGraphNode_StateMachine_0.ValetPassengerLowerBodyTurn.AnimStateNode_19.Transition_WtoS
  nodes [Output Animation Pose(Result=Blend Poses by bool(
     BlendPose_0=Sequence Player 'ValetVehicle_PL_DriveByTurnWtoS'() {bLoopAnimation=false, sequence=/Valet/Animation/Player/Basic/PassengerLeft/ValetVehicle_PL_DriveByTurnWtoS}
     BlendPose_1=Sequence Player 'ValetVehicle_DriveByTurnWtoS'() {bLoopAnimation=false, sequence=/Valet/Animation/Player/Basic/ValetVehicle_DriveByTurnWtoS}
     bActiveValue=Get bIsPassengerBackLeft()
    ) {blendType=Linear})]
 }
 Transition_NtoW {
  path=VehicleLowerBodyOverride.AnimGraphNode_StateMachine_0.ValetPassengerLowerBodyTurn.AnimStateNode_20.Transition_NtoW
  nodes [Output Animation Pose(Result=Blend Poses by bool(
     BlendPose_0=Sequence Player 'ValetVehicle_PL_DriveByTurnNtoW'() {bLoopAnimation=false, sequence=/Valet/Animation/Player/Basic/PassengerLeft/ValetVehicle_PL_DriveByTurnNtoW}
     BlendPose_1=Sequence Player 'ValetVehicle_DriveByTurnNtoW'() {bLoopAnimation=false, sequence=/Valet/Animation/Player/Basic/ValetVehicle_DriveByTurnNtoW}
     bActiveValue=Get bIsPassengerBackLeft()
    ) {blendType=Linear})]
 }
 Transition_NtoE {
  path=VehicleLowerBodyOverride.AnimGraphNode_StateMachine_0.ValetPassengerLowerBodyTurn.AnimStateNode_8.Transition_NtoE
  nodes [Output Animation Pose(Result=Blend Poses by bool(
     BlendPose_0=Sequence Player 'ValetVehicle_PL_DriveByTurnNtoE'() {bLoopAnimation=false, sequence=/Valet/Animation/Player/Basic/PassengerLeft/ValetVehicle_PL_DriveByTurnNtoE}
     BlendPose_1=Sequence Player 'ValetVehicle_DriveByTurnNtoE'() {bLoopAnimation=false, sequence=/Valet/Animation/Player/Basic/ValetVehicle_DriveByTurnNtoE}
     bActiveValue=Get bIsPassengerBackLeft()
    ) {blendType=Linear})]
 }
 PassengerPose_E {
  path=VehicleLowerBodyOverride.AnimGraphNode_StateMachine_0.ValetPassengerLowerBodyTurn.AnimStateNode_9.PassengerPose_E
  nodes [
   n1 = Sequence Player 'ValetVehicle_PL_DriveBy_E_Pose'() {sequence=/Valet/Animation/Player/Basic/PassengerLeft/ValetVehicle_PL_DriveBy_E_Pose}
   n2 = Sequence Player 'ValetVehicle_Passenger_DriveBy_E_Pose'() {sequence=/Valet/Animation/Player/Basic/ValetVehicle_Passenger_DriveBy_E_Pose}
   n3 = Get bIsPassengerBackLeft()
   n4 = AimOffset Player 'ShoppingCart_Passenger_LowerBodyAO'(
    X=Get AimRGTDeltaAngleDegrees() -> (AimRGTDeltaAngleDegrees:float)
    BasePose=Rotate Root Bone(BasePose=Blend Poses by bool(BlendPose_0=@n1, BlendPose_1=@n2, bActiveValue=@n3) {blendType=Linear}, Yaw=-90, MeshToComponent="0,0,0")
    Alpha=1
   ) {blendSpace=&ShoppingCart_Passenger_LowerBodyAO, alphaBoolBlend {blendOption=Linear}}
   Output Animation Pose(Result=Rotate Root Bone(BasePose=@n4, Yaw=90, MeshToComponent="0,0,0"))
  ]
 }
 AnimationTransitionGraph_0 {
  path=VehicleLowerBodyOverride.AnimGraphNode_StateMachine_0.ValetPassengerLowerBodyTurn.AnimStateTransitionNode_0.AnimationTransitionGraph_0
  nodes [Result(bCanEnterTransition=NOT Boolean(A=Get bAimRGT() -> (bAimRGT:bool)))]
 }
 AnimationTransitionGraph_1 {
  path=VehicleLowerBodyOverride.AnimGraphNode_StateMachine_0.ValetPassengerLowerBodyTurn.AnimStateTransitionNode_1.AnimationTransitionGraph_1
  nodes [Result()]
 }
 AnimationTransitionGraph_0 {
  path=VehicleLowerBodyOverride.AnimGraphNode_StateMachine_0.ValetPassengerLowerBodyTurn.AnimStateTransitionNode_10.AnimationTransitionGraph_0
  nodes [Result(bCanEnterTransition=Get bAimRGT())]
 }
 AnimationTransitionGraph_0 {
  path=VehicleLowerBodyOverride.AnimGraphNode_StateMachine_0.ValetPassengerLowerBodyTurn.AnimStateTransitionNode_11.AnimationTransitionGraph_0
  nodes [Result(bCanEnterTransition=Get bAimLFT() -> (bAimLFT:bool))]
 }
 AnimationTransitionGraph_1 {
  path=VehicleLowerBodyOverride.AnimGraphNode_StateMachine_0.ValetPassengerLowerBodyTurn.AnimStateTransitionNode_2.AnimationTransitionGraph_1
  nodes [Result()]
 }
 AnimationTransitionGraph_0 {
  path=VehicleLowerBodyOverride.AnimGraphNode_StateMachine_0.ValetPassengerLowerBodyTurn.AnimStateTransitionNode_25.AnimationTransitionGraph_0
  nodes [Result(bCanEnterTransition=Get bAimRGT())]
 }
 AnimationTransitionGraph_0 {
  path=VehicleLowerBodyOverride.AnimGraphNode_StateMachine_0.ValetPassengerLowerBodyTurn.AnimStateTransitionNode_26.AnimationTransitionGraph_0
  nodes [Result()]
 }
 AnimationTransitionGraph_1 {
  path=VehicleLowerBodyOverride.AnimGraphNode_StateMachine_0.ValetPassengerLowerBodyTurn.AnimStateTransitionNode_27.AnimationTransitionGraph_1
  nodes [Result(bCanEnterTransition=Get bAimBWD() -> (bAimBWD:bool))]
 }
 AnimationTransitionGraph_1 {
  path=VehicleLowerBodyOverride.AnimGraphNode_StateMachine_0.ValetPassengerLowerBodyTurn.AnimStateTransitionNode_28.AnimationTransitionGraph_1
  nodes [Result(bCanEnterTransition=NOT Boolean(A=Get bAimBWD()))]
 }
 AnimationTransitionGraph_1 {
  path=VehicleLowerBodyOverride.AnimGraphNode_StateMachine_0.ValetPassengerLowerBodyTurn.AnimStateTransitionNode_29.AnimationTransitionGraph_1
  nodes [Result(bCanEnterTransition=Get bAimLFT())]
 }
 AnimationTransitionGraph_0 {
  path=VehicleLowerBodyOverride.AnimGraphNode_StateMachine_0.ValetPassengerLowerBodyTurn.AnimStateTransitionNode_3.AnimationTransitionGraph_0
  nodes [Result()]
 }
 AnimationTransitionGraph_1 {
  path=VehicleLowerBodyOverride.AnimGraphNode_StateMachine_0.ValetPassengerLowerBodyTurn.AnimStateTransitionNode_30.AnimationTransitionGraph_1
  nodes [Result(bCanEnterTransition=NOT Boolean(A=Get bAimLFT()))]
 }
 AnimationTransitionGraph_0 {
  path=VehicleLowerBodyOverride.AnimGraphNode_StateMachine_0.ValetPassengerLowerBodyTurn.AnimStateTransitionNode_31.AnimationTransitionGraph_0
  nodes [Result(bCanEnterTransition=Get bAimFWD() -> (bAimFWD:bool))]
 }
 AnimationTransitionGraph_0 {
  path=VehicleLowerBodyOverride.AnimGraphNode_StateMachine_0.ValetPassengerLowerBodyTurn.AnimStateTransitionNode_32.AnimationTransitionGraph_0
  nodes [Result(bCanEnterTransition=NOT Boolean(A=Get bAimFWD()))]
 }
 AnimationTransitionGraph_0 {
  path=VehicleLowerBodyOverride.AnimGraphNode_StateMachine_0.ValetPassengerLowerBodyTurn.AnimStateTransitionNode_35.AnimationTransitionGraph_0
  nodes [Result(bCanEnterTransition=Get bAimFWD())]
 }
 AnimationTransitionGraph_0 {
  path=VehicleLowerBodyOverride.AnimGraphNode_StateMachine_0.ValetPassengerLowerBodyTurn.AnimStateTransitionNode_36.AnimationTransitionGraph_0
  nodes [Result(bCanEnterTransition=NOT Boolean(A=Get bAimFWD()))]
 }
 AnimationTransitionGraph_0 {
  path=VehicleLowerBodyOverride.AnimGraphNode_StateMachine_0.ValetPassengerLowerBodyTurn.AnimStateTransitionNode_37.AnimationTransitionGraph_0
  nodes [Result(bCanEnterTransition=Get bAimRGT())]
 }
 AnimationTransitionGraph_1 {
  path=VehicleLowerBodyOverride.AnimGraphNode_StateMachine_0.ValetPassengerLowerBodyTurn.AnimStateTransitionNode_38.AnimationTransitionGraph_1
  nodes [Result()]
 }
 AnimationTransitionGraph_1 {
  path=VehicleLowerBodyOverride.AnimGraphNode_StateMachine_0.ValetPassengerLowerBodyTurn.AnimStateTransitionNode_39.AnimationTransitionGraph_1
  nodes [Result(bCanEnterTransition=Get bAimBWD())]
 }
 AnimationTransitionGraph_0 {
  path=VehicleLowerBodyOverride.AnimGraphNode_StateMachine_0.ValetPassengerLowerBodyTurn.AnimStateTransitionNode_4.AnimationTransitionGraph_0
  nodes [Result(bCanEnterTransition=NOT Boolean(A=Get bAimLFT()))]
 }
 AnimationTransitionGraph_0 {
  path=VehicleLowerBodyOverride.AnimGraphNode_StateMachine_0.ValetPassengerLowerBodyTurn.AnimStateTransitionNode_40.AnimationTransitionGraph_0
  nodes [Result()]
 }
 AnimationTransitionGraph_0 {
  path=VehicleLowerBodyOverride.AnimGraphNode_StateMachine_0.ValetPassengerLowerBodyTurn.AnimStateTransitionNode_41.AnimationTransitionGraph_0
  nodes [Result(bCanEnterTransition=Get bAimLFT())]
 }
 AnimationTransitionGraph_0 {
  path=VehicleLowerBodyOverride.AnimGraphNode_StateMachine_0.ValetPassengerLowerBodyTurn.AnimStateTransitionNode_42.AnimationTransitionGraph_0
  nodes [Result()]
 }
 AnimationTransitionGraph_0 {
  path=VehicleLowerBodyOverride.AnimGraphNode_StateMachine_0.ValetPassengerLowerBodyTurn.AnimStateTransitionNode_5.AnimationTransitionGraph_0
  nodes [Result(bCanEnterTransition=NOT Boolean(A=Get bAimBWD()))]
 }
 AnimationTransitionGraph_1 {
  path=VehicleLowerBodyOverride.AnimGraphNode_StateMachine_0.ValetPassengerLowerBodyTurn.AnimStateTransitionNode_6.AnimationTransitionGraph_1
  nodes [Result(bCanEnterTransition=NOT Boolean(A=Get bAimRGT()))]
 }
 AnimationTransitionGraph_0 {
  path=VehicleLowerBodyOverride.AnimGraphNode_StateMachine_0.ValetPassengerLowerBodyTurn.AnimStateTransitionNode_7.AnimationTransitionGraph_0
  nodes [Result()]
 }
 AnimationTransitionGraph_1 {
  path=VehicleLowerBodyOverride.AnimGraphNode_StateMachine_0.ValetPassengerLowerBodyTurn.AnimStateTransitionNode_8.AnimationTransitionGraph_1
  nodes [Result(bCanEnterTransition=Get bAimBWD())]
 }
 AnimationTransitionGraph_1 {
  path=VehicleLowerBodyOverride.AnimGraphNode_StateMachine_0.ValetPassengerLowerBodyTurn.AnimStateTransitionNode_9.AnimationTransitionGraph_1
  nodes [Result(bCanEnterTransition=Get bAimFWD())]
 }
 VehicleSplitBodyOverride {nodes [
   Input Pose() {name=InPoseSplitBody}
   Input Pose() {name=InPoseUpperAndLowerBody}
   Input Pose() {name=InPoseUpperAndLowerBodyRemoveRoot}
   Output Pose() {layerGroup=Vehicle Override}
  ]}
]
variables [
 bIsBasicTruck {type=boolean}
 bIsSportsCar {type=boolean}
 Fort Player Pawn {type=FortPlayerPawn}
]
