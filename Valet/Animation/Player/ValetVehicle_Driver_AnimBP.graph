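# Player vehicle driver AnimBP that supplies the VehicleFinalPoseOverride layer, pulling most animations from CurrentDriverAnimSet.
# A Locomotion state machine handles idle with steering poses, driving, boosting, braking and reversing, with notifies disabling head aim while reversing.
# Additives then layer backpack, override pose, female correction and forward/backward impact reactions before the FullBody slot.
# Leg IK solves feet and hands, with hands following socket-based world-space targets, and weapon_r is scaled to zero.
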AnimBlueprint /Valet/Animation/Player/ValetVehicle_Driver_AnimBP
target_skeleton=&Fortnite_M_Avg_Player_Skeleton
graphs [
 AnimGraph {nodes [Output Pose() {layerGroup=None}]}
 EventGraph {nodes [
   AnimNotify_OnEnableAimOffset() {is=title K2Node_Event}
   > Set AimOffsetEnable(AimOffsetEnable:bool=true) -> (Output_Get:bool)
   AnimNotify_OnDisableAimOffset() {is=title K2Node_Event}
   > Set AimOffsetEnable(AimOffsetEnable=false)
  ]}
 VehicleFinalPoseOverride {nodes [
   Input Pose() {name=InFinalPose}
   AnimGraphNode_StateMachine_0 = Locomotion() {is=title AnimGraphNode_StateMachine}
   n1 = Break Fort Anim Input Vehicle Driver Anim Asset!(FortAnimInput_VehicleDriverAnimAsset:"Fort Anim Input Vehicle Driver Anim Asset Structure (by ref)"=Get CurrentDriverAnimSet() -> (CurrentDriverAnimSet:Fort Anim Input Vehicle Driver Anim Asset Structure)) -> (
    DriveNPose:Anim Sequence Object Reference
    DriveEPose:Anim Sequence Object Reference
    DriveWPose:Anim Sequence Object Reference
    DriveNAdditivePose:Anim Sequence Object Reference
    DriveFastAdditivePose:Anim Sequence Object Reference
    OverrideDriverPose:Anim Sequence Object Reference
    DriveIdle:Anim Sequence Object Reference
    DriveIdleFastAdditive:Anim Sequence Object Reference
    DriverHeadAimOffset:Aim Offset Blend Space Object Reference
    DriveNStart:Anim Sequence Object Reference
    Braking:Anim Sequence Object Reference
    BoostStart:Anim Sequence Object Reference
    BoostLoop:Anim Sequence Object Reference
    ReverseStart:Anim Sequence Object Reference
    ReverseLoop:Anim Sequence Object Reference
    ReverseEnd:Anim Sequence Object Reference
    ReturnToIdleTransition:Anim Sequence Object Reference
    PoseCorrectionAdditive:Anim Sequence Object Reference
    CollisionN:Anim Sequence Object Reference
    CollisionS:Anim Sequence Object Reference
    CollisionE:Anim Sequence Object Reference
    CollisionW:Anim Sequence Object Reference
   ) {is=title K2Node_BreakStruct}
   n2 = Apply Additive(
    Base="AimOffset Player '(None)'"!(
     X:float=Get AimYawSmoothed() -> (AimYawSmoothed:float)
     Y:float=Get AimPitch()
     BlendSpace:Blend Space Object Reference=@n1.DriverHeadAimOffset
     BasePose:Pose Link Structure=@AnimGraphNode_StateMachine_0
     bAlphaBoolEnabled:bool=Get AimOffsetEnable() -> (AimOffsetEnable:bool)
    ) -> (Pose:Pose Link Structure) {alphaInputType=Bool, alphaBoolBlend {blendInTime=0.25, blendOption=Linear, blendOutTime=0.25}}
    Additive="Sequence Player (additive) 'MeatBall_Driver_Backpack_Override'"() {sequence=/Game/Animation/Game/MainPlayer/Vehicles/MeatBallVehicle/MeatBall_Driver_Backpack_Override}
   ) {alphaBoolBlend {blendOption=Linear}}
   n3 = Sequence Player!(Sequence:Anim Sequence Base Object Reference=@n1.OverrideDriverPose) -> (Pose)
   AnimGraphNode_StateMachine_2 = FullBodyAdditive() {is=title AnimGraphNode_StateMachine}
   n4 = Apply Additive(
    Base=Apply Additive!(
     Base=Apply Additive!(Base=@n2, Additive=@n3, Alpha:float) -> (Pose) {alphaBoolBlend {blendOption=Linear}}
     Additive=Sequence Player(Sequence=@n1.PoseCorrectionAdditive)
     bAlphaBoolEnabled:bool=Get bIsFemale() -> (bIsFemale:bool)
    ) -> (Pose) {alphaInputType=Bool, alphaBoolBlend {blendOption=Linear}}
    Additive=@AnimGraphNode_StateMachine_2
   ) {alphaBoolBlend {blendOption=Linear}}
   n5 = Leg IK!(
    ComponentPose:Component Space Pose Link Structure=Local To Component(LocalPose=Slot 'FullBody'(Source=@n4) {slotName=FullBody})
    Alpha:float=1
   ) -> (Pose:Component Space Pose Link Structure) {
    is=title AnimGraphNode_LegIK
    alphaBoolBlend {blendOption=Linear}
    legsDefinition [
     {
      bEnableKneeTwistCorrection=false
      bEnableRotationLimit=false
      footBoneForwardAxis=Y
      hingeRotationAxis=None
      minRotationAngle=15
      numBonesInLimb=2
      twistOffsetCurveName=None
      fKFootBone {boneName=foot_l}
      iKFootBone {boneName=ik_foot_l}
     }
     {
      bEnableKneeTwistCorrection=false
      bEnableRotationLimit=false
      footBoneForwardAxis=Y
      hingeRotationAxis=None
      minRotationAngle=15
      numBonesInLimb=2
      twistOffsetCurveName=None
      fKFootBone {boneName=foot_r}
      iKFootBone {boneName=ik_foot_r}
     }
    ]
   }
   n6 = Break Fort Anim Input Socket Based IKTarget!(FortAnimInput_SocketBasedIKTarget:"Fort Anim Input Socket Based IKTarget Structure (by ref)"=Get LeftHandIKTarget() -> (LeftHandIKTarget:Fort Anim Input Socket Based IKTarget Structure)) -> (SocketName:Name, AlphaCurveName:Name, TargetRotation:Rotator=~, TargetLocation:Vector=~, Alpha:float=~) {is=title K2Node_BreakStruct}
   n7 = "Transform (Modify) Bone - Bone: ik_hand_l"!(ComponentPose=@n5, Alpha=@n6.Alpha, Translation=@n6.TargetLocation, Rotation=@n6.TargetRotation) -> (Pose) {
    rotationMode=BMM_Replace
    rotationSpace=BCS_WorldSpace
    translationMode=BMM_Replace
    translationSpace=BCS_WorldSpace
    alphaBoolBlend {blendOption=Linear}
    boneToModify {boneName=ik_hand_l}
   }
   n8 = Break Fort Anim Input Socket Based IKTarget(FortAnimInput_SocketBasedIKTarget=Get RightHandIKTarget() -> (RightHandIKTarget:Fort Anim Input Socket Based IKTarget Structure)) -> (TargetRotation=~, TargetLocation=~, Alpha=~)
   n9 = "Transform (Modify) Bone - Bone: ik_hand_r"!(ComponentPose=@n7, Alpha=@n8.Alpha, Translation=@n8.TargetLocation, Rotation=@n8.TargetRotation) -> (Pose) {
    rotationMode=BMM_Replace
    rotationSpace=BCS_WorldSpace
    translationMode=BMM_Replace
    translationSpace=BCS_WorldSpace
    alphaBoolBlend {blendOption=Linear}
    boneToModify {boneName=ik_hand_r}
   }
   n10 = Leg IK(ComponentPose=@n9, Alpha=1) {
    alphaBoolBlend {blendOption=Linear}
    legsDefinition [{
      bEnableKneeTwistCorrection=true
      bEnableRotationLimit=false
      footBoneForwardAxis=Y
      hingeRotationAxis=None
      minRotationAngle=15
      numBonesInLimb=2
      twistOffsetCurveName=None
      fKFootBone {boneName=hand_l}
      iKFootBone {boneName=ik_hand_l}
     }]
   }
   n11 = Leg IK(ComponentPose=@n10, Alpha=1) {
    alphaBoolBlend {blendOption=Linear}
    legsDefinition [{
      bEnableKneeTwistCorrection=true
      bEnableRotationLimit=false
      footBoneForwardAxis=Y
      hingeRotationAxis=None
      minRotationAngle=15
      numBonesInLimb=2
      twistOffsetCurveName=None
      fKFootBone {boneName=hand_r}
      iKFootBone {boneName=ik_hand_r}
     }]
   }
   n12 = "Transform (Modify) Bone - Bone: weapon_r"!(ComponentPose=@n11, Alpha=1, Scale="0,0,0") -> (Pose) {
    lODThreshold=3
    scaleMode=BMM_Replace
    scaleSpace=BCS_BoneSpace
    alphaBoolBlend {blendOption=Linear}
    boneToModify {boneName=weapon_r}
    scale=<0 0 0>
   }
   Output Pose(Result=Component To Local(ComponentPose=@n12)) {layerGroup=VehicleOverride}
   AnimGraphNode_StateMachine_1 = Driver_Idle() {is=title AnimGraphNode_StateMachine}
   Save cached pose 'Idle_Pose'(Pose=@AnimGraphNode_StateMachine_1) {cacheName=Idle_Pose}
   n13 = "Sequence Evaluator (by time)"!(Sequence:Anim Sequence Base Object Reference=@n1) -> (Pose:Pose Link Structure) {is=title AnimGraphNode_SequenceEvaluator}
   n14 = "Sequence Evaluator (by time)"(Sequence=@n1.DriveEPose)
   n15 = Get bIsSteeringRight() -> (bIsSteeringRight:bool)
   Save cached pose 'SteerDirectionPose'(Pose=Blend Poses by bool(
     BlendPose_0="Sequence Evaluator (by time)"(Sequence=@n1.DriveWPose)
     BlendPose_1=Blend Poses by bool(BlendPose_0=@n14, BlendPose_1=@n13, BlendTime_0=0.35, BlendTime_1=0.35, bActiveValue=@n15) {blendType=Linear}
     BlendTime_0=0.35
     BlendTime_1=0.35
     bActiveValue=Get bIsSteeringLeft() -> (bIsSteeringLeft:bool)
    ) {blendType=Linear}) {cacheName=SteerDirectionPose}
  ]}
 Locomotion {
  entry=Default
  path=VehicleFinalPoseOverride.AnimGraphNode_StateMachine_0.Locomotion
  ids=id
  nodes [
   AnimStateNode_0 = Default() {
    is=title AnimStateNode
    stateEntered {guid=#1, notifyName=OnEnableAimOffset}
    stateFullyBlended {guid=#2, notifyName=OnEnableAimOffset}
   }
   AnimStateNode_1 = Drive_N_Start() {is=title AnimStateNode}
   AnimStateNode_12 = Boost_Loop() {is=title AnimStateNode}
   AnimStateNode_2 = Drive_N_Loop() {is=title AnimStateNode}
   AnimStateNode_3 = Reverse() {is=title AnimStateNode}
   AnimStateNode_4 = Braking() {is=title AnimStateNode}
   AnimStateNode_5 = Back_to_Idle_Transition() {is=title AnimStateNode}
   AnimStateNode_6 = Reverse_Start() {is=title AnimStateNode, stateEntered {guid=#3, notifyName=OnDisableAimOffset}}
   AnimStateNode_7 = Boost_Start() {is=title AnimStateNode}
   AnimStateNode_8 = Reverse_End() {is=title AnimStateNode}
   AnimStateTransitionNode_0 = AnimStateTransitionNode() {crossfadeDuration=0.15, priorityOrder=2}
   AnimStateTransitionNode_1 = AnimStateTransitionNode() {bAutomaticRuleBasedOnSequencePlayerInState=true, crossfadeDuration=0.35}
   AnimStateTransitionNode_10 = AnimStateTransitionNode() {bAutomaticRuleBasedOnSequencePlayerInState=true, crossfadeDuration=0.35}
   AnimStateTransitionNode_11 = AnimStateTransitionNode() {crossfadeDuration=0.1}
   AnimStateTransitionNode_12 = AnimStateTransitionNode() {bAutomaticRuleBasedOnSequencePlayerInState=true, priorityOrder=2}
   AnimStateTransitionNode_13 = AnimStateTransitionNode()
   AnimStateTransitionNode_14 = AnimStateTransitionNode()
   AnimStateTransitionNode_15 = AnimStateTransitionNode()
   AnimStateTransitionNode_16 = AnimStateTransitionNode()
   AnimStateTransitionNode_17 = AnimStateTransitionNode()
   AnimStateTransitionNode_18 = AnimStateTransitionNode()
   AnimStateTransitionNode_2 = AnimStateTransitionNode() {crossfadeDuration=0.1}
   AnimStateTransitionNode_20 = AnimStateTransitionNode() {bAutomaticRuleBasedOnSequencePlayerInState=true, crossfadeDuration=0.4}
   AnimStateTransitionNode_21 = AnimStateTransitionNode() {crossfadeDuration=0.35}
   AnimStateTransitionNode_22 = AnimStateTransitionNode()
   AnimStateTransitionNode_3 = AnimStateTransitionNode() {priorityOrder=2}
   AnimStateTransitionNode_4 = AnimStateTransitionNode() {bAutomaticRuleBasedOnSequencePlayerInState=true}
   AnimStateTransitionNode_5 = AnimStateTransitionNode() {crossfadeDuration=0.25}
   AnimStateTransitionNode_6 = AnimStateTransitionNode()
   AnimStateTransitionNode_7 = AnimStateTransitionNode()
   AnimStateTransitionNode_8 = AnimStateTransitionNode()
   AnimStateTransitionNode_9 = AnimStateTransitionNode() {bAutomaticRuleBasedOnSequencePlayerInState=true, crossfadeDuration=0.3}
  ]
  links [
   {
    dst=Boost_Start
    rule=VehicleFinalPoseOverride.AnimGraphNode_StateMachine_0.Locomotion.AnimStateTransitionNode_17.Transition
    src=Back_to_Idle_Transition
    via=AnimStateTransitionNode_17
   }
   {
    dst=Default
    rule=VehicleFinalPoseOverride.AnimGraphNode_StateMachine_0.Locomotion.AnimStateTransitionNode_9.Transition
    src=Back_to_Idle_Transition
    via=AnimStateTransitionNode_9
   }
   {
    dst=Drive_N_Loop
    rule=VehicleFinalPoseOverride.AnimGraphNode_StateMachine_0.Locomotion.AnimStateTransitionNode_21.Transition
    src=Boost_Loop
    via=AnimStateTransitionNode_21
   }
   {
    dst=Boost_Loop
    rule=VehicleFinalPoseOverride.AnimGraphNode_StateMachine_0.Locomotion.AnimStateTransitionNode_20.Transition
    src=Boost_Start
    via=AnimStateTransitionNode_20
   }
   {
    dst=Boost_Start
    rule=VehicleFinalPoseOverride.AnimGraphNode_StateMachine_0.Locomotion.AnimStateTransitionNode_18.Transition
    src=Braking
    via=AnimStateTransitionNode_18
   }
   {
    dst=Drive_N_Loop
    rule=VehicleFinalPoseOverride.AnimGraphNode_StateMachine_0.Locomotion.AnimStateTransitionNode_10.Transition
    src=Braking
    via=AnimStateTransitionNode_10
   }
   {
    dst=Drive_N_Start
    rule=VehicleFinalPoseOverride.AnimGraphNode_StateMachine_0.Locomotion.AnimStateTransitionNode_0.Transition
    src=Default
    via=AnimStateTransitionNode_0
   }
   {
    dst=Reverse_Start
    rule=VehicleFinalPoseOverride.AnimGraphNode_StateMachine_0.Locomotion.AnimStateTransitionNode_11.Transition
    src=Default
    via=AnimStateTransitionNode_11
   }
   {
    dst=Back_to_Idle_Transition
    rule=VehicleFinalPoseOverride.AnimGraphNode_StateMachine_0.Locomotion.AnimStateTransitionNode_8.Transition
    src=Drive_N_Loop
    via=AnimStateTransitionNode_8
   }
   {
    dst=Boost_Start
    rule=VehicleFinalPoseOverride.AnimGraphNode_StateMachine_0.Locomotion.AnimStateTransitionNode_15.Transition
    src=Drive_N_Loop
    via=AnimStateTransitionNode_15
   }
   {
    dst=Braking
    rule=VehicleFinalPoseOverride.AnimGraphNode_StateMachine_0.Locomotion.AnimStateTransitionNode_5.Transition
    src=Drive_N_Loop
    via=AnimStateTransitionNode_5
   }
   {
    dst=Reverse_Start
    rule=VehicleFinalPoseOverride.AnimGraphNode_StateMachine_0.Locomotion.AnimStateTransitionNode_2.Transition
    src=Drive_N_Loop
    via=AnimStateTransitionNode_2
   }
   {
    dst=Boost_Start
    rule=VehicleFinalPoseOverride.AnimGraphNode_StateMachine_0.Locomotion.AnimStateTransitionNode_16.Transition
    src=Drive_N_Start
    via=AnimStateTransitionNode_16
   }
   {
    dst=Default
    rule=VehicleFinalPoseOverride.AnimGraphNode_StateMachine_0.Locomotion.AnimStateTransitionNode_22.Transition
    src=Drive_N_Start
    via=AnimStateTransitionNode_22
   }
   {
    dst=Drive_N_Loop
    rule=VehicleFinalPoseOverride.AnimGraphNode_StateMachine_0.Locomotion.AnimStateTransitionNode_1.Transition
    src=Drive_N_Start
    via=AnimStateTransitionNode_1
   }
   {
    dst=Boost_Start
    rule=VehicleFinalPoseOverride.AnimGraphNode_StateMachine_0.Locomotion.AnimStateTransitionNode_13.Transition
    src=Reverse
    via=AnimStateTransitionNode_13
   }
   {
    dst=Drive_N_Start
    rule=VehicleFinalPoseOverride.AnimGraphNode_StateMachine_0.Locomotion.AnimStateTransitionNode_7.Transition
    src=Reverse
    via=AnimStateTransitionNode_7
   }
   {
    dst=Reverse_End
    rule=VehicleFinalPoseOverride.AnimGraphNode_StateMachine_0.Locomotion.AnimStateTransitionNode_3.Transition
    src=Reverse
    via=AnimStateTransitionNode_3
   }
   {
    dst=Boost_Start
    rule=VehicleFinalPoseOverride.AnimGraphNode_StateMachine_0.Locomotion.AnimStateTransitionNode_14.Transition
    src=Reverse_End
    via=AnimStateTransitionNode_14
   }
   {
    dst=Default
    rule=VehicleFinalPoseOverride.AnimGraphNode_StateMachine_0.Locomotion.AnimStateTransitionNode_4.Transition
    src=Reverse_End
    via=AnimStateTransitionNode_4
   }
   {
    dst=Boost_Start
    rule=VehicleFinalPoseOverride.AnimGraphNode_StateMachine_0.Locomotion.AnimStateTransitionNode_6.Transition
    src=Reverse_Start
    via=AnimStateTransitionNode_6
   }
   {
    dst=Reverse
    rule=VehicleFinalPoseOverride.AnimGraphNode_StateMachine_0.Locomotion.AnimStateTransitionNode_12.Transition
    src=Reverse_Start
    via=AnimStateTransitionNode_12
   }
  ]
 }
 Default {
  path=VehicleFinalPoseOverride.AnimGraphNode_StateMachine_0.Locomotion.AnimStateNode_0.Default
  nodes [Output Animation Pose(Result=Use cached pose 'Idle_Pose'())]
 }
 Drive_N_Start {
  path=VehicleFinalPoseOverride.AnimGraphNode_StateMachine_0.Locomotion.AnimStateNode_1.Drive_N_Start
  nodes [
   n1 = Break Fort Anim Input Vehicle Driver Anim Asset(FortAnimInput_VehicleDriverAnimAsset=Get CurrentDriverAnimSet())
   n2 = Get DriveStartPlayRate() -> (DriveStartPlayRate:float)
   Output Animation Pose(Result=Apply Additive(Base=Use cached pose 'Idle_Pose'(), Additive=Sequence Player!(Sequence=@n1.DriveNStart, PlayRate:float=@n2) -> (Pose) {bLoopAnimation=false}) {alphaBoolBlend {blendOption=Linear}})
  ]
 }
 Boost_Loop {
  path=VehicleFinalPoseOverride.AnimGraphNode_StateMachine_0.Locomotion.AnimStateNode_12.Boost_Loop
  nodes [
   n1 = Get CurrentDriverAnimSet()
   n2 = AimOffset Player 'DriveForwardLeans_AO_GolfCart'!(
    X:float=Get SteeringAngle() -> (SteeringAngle:float)
    BasePose:Pose Link Structure=Sequence Player!(Sequence=Break Fort Anim Input Vehicle Driver Anim Asset(FortAnimInput_VehicleDriverAnimAsset=@n1).BoostLoop) -> (Pose)
    Alpha:float=1
   ) -> (Pose:Pose Link Structure) {blendSpace=/Game/Animation/Game/MainPlayer/Vehicles/GolfCart/AO/DriveForwardLeans_AO_GolfCart, alphaBoolBlend {blendOption=Linear}}
   Output Animation Pose(Result=@n2)
  ]
 }
 Drive_N_Loop {
  path=VehicleFinalPoseOverride.AnimGraphNode_StateMachine_0.Locomotion.AnimStateNode_2.Drive_N_Loop
  nodes [
   n1 = Break Fort Anim Input Vehicle Driver Anim Asset(FortAnimInput_VehicleDriverAnimAsset=Get CurrentDriverAnimSet())
   n2 = Sequence Player(Sequence=@n1.DriveFastAdditivePose)
   n3 = Sequence Player(Sequence=@n1.DriveNAdditivePose)
   n4 = Get bIsMovingFast() -> (bIsMovingFast:bool)
   Output Animation Pose(Result=Apply Additive(
     Base=Use cached pose 'Idle_Pose'()
     Additive=Blend Poses by bool(BlendPose_0=@n2, BlendPose_1=@n3, BlendTime_0=0.4, BlendTime_1=0.3, bActiveValue=@n4) {blendType=Linear}
    ) {alphaBoolBlend {blendOption=Linear}})
  ]
 }
 Reverse {
  path=VehicleFinalPoseOverride.AnimGraphNode_StateMachine_0.Locomotion.AnimStateNode_3.Reverse
  nodes [
   n1 = Get CurrentDriverAnimSet()
   Output Animation Pose(Result=Sequence Player(Sequence=Break Fort Anim Input Vehicle Driver Anim Asset(FortAnimInput_VehicleDriverAnimAsset=@n1).ReverseLoop))
  ]
 }
 Braking {
  path=VehicleFinalPoseOverride.AnimGraphNode_StateMachine_0.Locomotion.AnimStateNode_4.Braking
  nodes [
   n1 = Get CurrentDriverAnimSet()
   Output Animation Pose(Result=Sequence Player!(
     Sequence=Break Fort Anim Input Vehicle Driver Anim Asset(FortAnimInput_VehicleDriverAnimAsset=@n1).Braking
     StartPosition:float=Get StartPosition() -> (StartPosition:double)
    ) -> (Pose) {bLoopAnimation=false})
  ]
 }
 Back_to_Idle_Transition {
  path=VehicleFinalPoseOverride.AnimGraphNode_StateMachine_0.Locomotion.AnimStateNode_5.Back_to_Idle_Transition
  nodes [
   n1 = Get CurrentDriverAnimSet()
   Output Animation Pose(Result=Sequence Player!(Sequence=Break Fort Anim Input Vehicle Driver Anim Asset(FortAnimInput_VehicleDriverAnimAsset=@n1).ReturnToIdleTransition) -> (Pose) {bLoopAnimation=false})
  ]
 }
 Reverse_Start {
  path=VehicleFinalPoseOverride.AnimGraphNode_StateMachine_0.Locomotion.AnimStateNode_6.Reverse_Start
  nodes [
   n1 = Get CurrentDriverAnimSet()
   Output Animation Pose(Result=Sequence Player(Sequence=Break Fort Anim Input Vehicle Driver Anim Asset(FortAnimInput_VehicleDriverAnimAsset=@n1).ReverseStart) {bLoopAnimation=false})
  ]
 }
 Boost_Start {
  path=VehicleFinalPoseOverride.AnimGraphNode_StateMachine_0.Locomotion.AnimStateNode_7.Boost_Start
  nodes [
   n1 = Get CurrentDriverAnimSet()
   Output Animation Pose(Result=Sequence Player(Sequence=Break Fort Anim Input Vehicle Driver Anim Asset(FortAnimInput_VehicleDriverAnimAsset=@n1).BoostStart) {bLoopAnimation=false})
  ]
 }
 Reverse_End {
  path=VehicleFinalPoseOverride.AnimGraphNode_StateMachine_0.Locomotion.AnimStateNode_8.Reverse_End
  nodes [
   n1 = Get CurrentDriverAnimSet()
   Output Animation Pose(Result=Sequence Player(Sequence=Break Fort Anim Input Vehicle Driver Anim Asset(FortAnimInput_VehicleDriverAnimAsset=@n1).ReverseEnd) {bLoopAnimation=false})
  ]
 }
 Transition {
  path=VehicleFinalPoseOverride.AnimGraphNode_StateMachine_0.Locomotion.AnimStateTransitionNode_0.Transition
  nodes [Result(bCanEnterTransition=Get bIsMovingForward() -> (bIsMovingForward:bool))]
 }
 Transition {path=VehicleFinalPoseOverride.AnimGraphNode_StateMachine_0.Locomotion.AnimStateTransitionNode_1.Transition, nodes [Result()]}
 Transition {path=VehicleFinalPoseOverride.AnimGraphNode_StateMachine_0.Locomotion.AnimStateTransitionNode_10.Transition, nodes [Result()]}
 Transition {
  path=VehicleFinalPoseOverride.AnimGraphNode_StateMachine_0.Locomotion.AnimStateTransitionNode_11.Transition
  nodes [Result(bCanEnterTransition=Get bIsReversing() -> (bIsReversing:bool))]
 }
 Transition {path=VehicleFinalPoseOverride.AnimGraphNode_StateMachine_0.Locomotion.AnimStateTransitionNode_12.Transition, nodes [Result()]}
 Transition {
  path=VehicleFinalPoseOverride.AnimGraphNode_StateMachine_0.Locomotion.AnimStateTransitionNode_13.Transition
  nodes [Result(bCanEnterTransition=Get bIsBoosting() -> (bIsBoosting:bool))]
 }
 Transition {
  path=VehicleFinalPoseOverride.AnimGraphNode_StateMachine_0.Locomotion.AnimStateTransitionNode_14.Transition
  nodes [Result(bCanEnterTransition=Get bIsBoosting())]
 }
 Transition {
  path=VehicleFinalPoseOverride.AnimGraphNode_StateMachine_0.Locomotion.AnimStateTransitionNode_15.Transition
  nodes [Result(bCanEnterTransition=Get bIsBoosting())]
 }
 Transition {
  path=VehicleFinalPoseOverride.AnimGraphNode_StateMachine_0.Locomotion.AnimStateTransitionNode_16.Transition
  nodes [Result(bCanEnterTransition=Get bIsBoosting())]
 }
 Transition {
  path=VehicleFinalPoseOverride.AnimGraphNode_StateMachine_0.Locomotion.AnimStateTransitionNode_17.Transition
  nodes [Result(bCanEnterTransition=Get bIsBoosting())]
 }
 Transition {
  path=VehicleFinalPoseOverride.AnimGraphNode_StateMachine_0.Locomotion.AnimStateTransitionNode_18.Transition
  nodes [Result(bCanEnterTransition=Get bIsBoosting())]
 }
 Transition {
  path=VehicleFinalPoseOverride.AnimGraphNode_StateMachine_0.Locomotion.AnimStateTransitionNode_2.Transition
  nodes [Result(bCanEnterTransition=Get bIsReversing())]
 }
 Transition {path=VehicleFinalPoseOverride.AnimGraphNode_StateMachine_0.Locomotion.AnimStateTransitionNode_20.Transition, nodes [Result()]}
 Transition {
  path=VehicleFinalPoseOverride.AnimGraphNode_StateMachine_0.Locomotion.AnimStateTransitionNode_21.Transition
  nodes [Result(bCanEnterTransition=NOT Boolean(A=Get bIsBoosting()))]
 }
 Transition {
  path=VehicleFinalPoseOverride.AnimGraphNode_StateMachine_0.Locomotion.AnimStateTransitionNode_22.Transition
  nodes [Result(bCanEnterTransition=Get bIsReversing())]
 }
 Transition {
  path=VehicleFinalPoseOverride.AnimGraphNode_StateMachine_0.Locomotion.AnimStateTransitionNode_3.Transition
  nodes [
   Result(bCanEnterTransition=NOT Boolean(A=Get bIsReversing()))
   Get bIsMoving() -> (bIsMoving:bool)
  ]
 }
 Transition {path=VehicleFinalPoseOverride.AnimGraphNode_StateMachine_0.Locomotion.AnimStateTransitionNode_4.Transition, nodes [Result()]}
 Transition {
  path=VehicleFinalPoseOverride.AnimGraphNode_StateMachine_0.Locomotion.AnimStateTransitionNode_5.Transition
  nodes [Result(bCanEnterTransition=Get bIsBraking() -> (bIsBraking:bool))]
 }
 Transition {
  path=VehicleFinalPoseOverride.AnimGraphNode_StateMachine_0.Locomotion.AnimStateTransitionNode_6.Transition
  nodes [Result(bCanEnterTransition=Get bIsBoosting())]
 }
 Transition {
  path=VehicleFinalPoseOverride.AnimGraphNode_StateMachine_0.Locomotion.AnimStateTransitionNode_7.Transition
  nodes [Result(bCanEnterTransition=Get bIsMovingForward())]
 }
 Transition {
  path=VehicleFinalPoseOverride.AnimGraphNode_StateMachine_0.Locomotion.AnimStateTransitionNode_8.Transition
  nodes [Result(bCanEnterTransition=NOT Boolean(A=Get bIsMoving()))]
 }
 Transition {path=VehicleFinalPoseOverride.AnimGraphNode_StateMachine_0.Locomotion.AnimStateTransitionNode_9.Transition, nodes [Result()]}
 Driver_Idle {
  entry=Idle_Leans
  path=VehicleFinalPoseOverride.AnimGraphNode_StateMachine_1.Driver_Idle
  ids=id
  nodes [AnimStateNode_0 = Idle_Leans() {is=title AnimStateNode}]
 }
 Idle_Leans {
  path=VehicleFinalPoseOverride.AnimGraphNode_StateMachine_1.Driver_Idle.AnimStateNode_0.Idle_Leans
  nodes [
   n1 = Break Fort Anim Input Vehicle Driver Anim Asset(FortAnimInput_VehicleDriverAnimAsset=Get CurrentDriverAnimSet())
   n2 = Apply Additive(Base=Use cached pose 'SteerDirectionPose'(), Additive=Sequence Player(Sequence=@n1.DriveIdleFastAdditive)) {alphaBoolBlend {blendOption=Linear}}
   n3 = Use cached pose 'SteerDirectionPose'()
   n4 = Get bIsMovingFast()
   Output Animation Pose(Result=Blend Poses by bool(
     BlendPose_0=Blend Poses by bool(BlendPose_0=@n2, BlendPose_1=@n3, BlendTime_0=0.25, BlendTime_1=0.25, bActiveValue=@n4) {blendType=Linear}
     BlendPose_1=Sequence Player(Sequence=@n1.DriveIdle) {groupName=Locomotion, method=SyncGroup}
     BlendTime_0=0.35
     BlendTime_1=0.35
     bActiveValue=Get bIsMoving()
    ) {blendType=Linear})
  ]
 }
 FullBodyAdditive {
  entry=Default
  path=VehicleFinalPoseOverride.AnimGraphNode_StateMachine_2.FullBodyAdditive
  ids=id
  nodes [
   AnimStateNode_0 = Default()
   AnimStateNode_1 = Collision_N() {is=title AnimStateNode}
   AnimStateNode_2 = Collision_S() {is=title AnimStateNode}
   AnimStateTransitionNode_0 = AnimStateTransitionNode() {crossfadeDuration=0}
   AnimStateTransitionNode_1 = AnimStateTransitionNode() {crossfadeDuration=0}
   AnimStateTransitionNode_2 = AnimStateTransitionNode() {bAutomaticRuleBasedOnSequencePlayerInState=true, crossfadeDuration=0}
   AnimStateTransitionNode_3 = AnimStateTransitionNode() {bAutomaticRuleBasedOnSequencePlayerInState=true, crossfadeDuration=0}
  ]
  links [
   {
    dst=Default
    rule=VehicleFinalPoseOverride.AnimGraphNode_StateMachine_2.FullBodyAdditive.AnimStateTransitionNode_2.Transition
    src=Collision_N
    via=AnimStateTransitionNode_2
   }
   {
    dst=Default
    rule=VehicleFinalPoseOverride.AnimGraphNode_StateMachine_2.FullBodyAdditive.AnimStateTransitionNode_3.Transition
    src=Collision_S
    via=AnimStateTransitionNode_3
   }
   {
    dst=Collision_N
    rule=VehicleFinalPoseOverride.AnimGraphNode_StateMachine_2.FullBodyAdditive.AnimStateTransitionNode_0.Transition
    src=Default
    via=AnimStateTransitionNode_0
   }
   {
    dst=Collision_S
    rule=VehicleFinalPoseOverride.AnimGraphNode_StateMachine_2.FullBodyAdditive.AnimStateTransitionNode_1.Transition
    src=Default
    via=AnimStateTransitionNode_1
   }
  ]
 }
 Default {
  path=VehicleFinalPoseOverride.AnimGraphNode_StateMachine_2.FullBodyAdditive.AnimStateNode_0.Default
  nodes [Output Animation Pose(Result=Additive Identity Pose())]
 }
 Collision_N {
  path=VehicleFinalPoseOverride.AnimGraphNode_StateMachine_2.FullBodyAdditive.AnimStateNode_1.Collision_N
  nodes [Output Animation Pose(Result="Sequence Player (additive) 'MeatBall_Driver_Collision_N'"() {bLoopAnimation=false, sequence=/Game/Animation/Game/MainPlayer/Vehicles/MeatBallVehicle/MeatBall_Driver_Collision_N})]
 }
 Collision_S {
  path=VehicleFinalPoseOverride.AnimGraphNode_StateMachine_2.FullBodyAdditive.AnimStateNode_2.Collision_S
  nodes [Output Animation Pose(Result="Sequence Player (additive) 'MeatBall_Driver_Collision_S'"() {bLoopAnimation=false, sequence=/Game/Animation/Game/MainPlayer/Vehicles/MeatBallVehicle/MeatBall_Driver_Collision_S})]
 }
 Transition {
  path=VehicleFinalPoseOverride.AnimGraphNode_StateMachine_2.FullBodyAdditive.AnimStateTransitionNode_0.Transition
  nodes [Result(bCanEnterTransition=Break Fort Anim Input Velocity Impact!(FortAnimInput_VelocityImpact:"Fort Anim Input Velocity Impact Structure (by ref)"=Get VelocityImpact() -> (VelocityImpact:Fort Anim Input Velocity Impact Structure)) -> (
     DeltaVelocityThreshold:Vector=~
     ImpactScale:Vector=~
     ImpactLimitX:Input Range Structure
     ImpactLimitY:Input Range Structure
     ImpactLimitZ:Input Range Structure
     SpringInterpolatorX:Float RK4Spring Interpolator Structure
     SpringInterpolatorY:Float RK4Spring Interpolator Structure
     SpringInterpolatorZ:Float RK4Spring Interpolator Structure
     bTestVelocity:bool=~
     TestVelocity:Vector=~
     bIsForwardImpact:bool=~
     bIsBackwardImpact:bool=~
     bIsLeftImpact:bool=~
     bIsRightImpact:bool=~
     bIsUpImpact:bool=~
     bIsDownImpact:bool=~
    ) {is=title K2Node_BreakStruct}.bIsForwardImpact)]
 }
 Transition {
  path=VehicleFinalPoseOverride.AnimGraphNode_StateMachine_2.FullBodyAdditive.AnimStateTransitionNode_1.Transition
  nodes [Result(bCanEnterTransition=Break Fort Anim Input Velocity Impact(FortAnimInput_VelocityImpact=Get VelocityImpact()) -> (
     DeltaVelocityThreshold=~
     ImpactScale=~
     bTestVelocity=~
     TestVelocity=~
     bIsForwardImpact=~
     bIsBackwardImpact=~
     bIsLeftImpact=~
     bIsRightImpact=~
     bIsUpImpact=~
     bIsDownImpact=~
    ).bIsBackwardImpact)]
 }
 Transition {path=VehicleFinalPoseOverride.AnimGraphNode_StateMachine_2.FullBodyAdditive.AnimStateTransitionNode_2.Transition, nodes [Result()]}
 Transition {path=VehicleFinalPoseOverride.AnimGraphNode_StateMachine_2.FullBodyAdditive.AnimStateTransitionNode_3.Transition, nodes [Result()]}
 VehicleFullBodyOverride {nodes [
   Input Pose() {name=InPoseFullBody}
   Output Pose() {layerGroup=VehicleOverride}
  ]}
 VehicleLowerBodyOverride {nodes [
   Input Pose() {name=InPoseLowerBodyDefault}
   Input Pose() {name=InPoseUpperBody}
   Output Pose() {layerGroup=VehicleOverride}
  ]}
 VehicleSplitBodyOverride {nodes [
   Input Pose() {name=InPoseSplitBody}
   Input Pose() {name=InPoseUpperAndLowerBody}
   Input Pose() {name=InPoseUpperAndLowerBodyRemoveRoot}
   Output Pose() {layerGroup=Vehicle Override}
  ]}
]
variables [
 AimOffsetEnable {type=boolean}
 StartPosition {type=number, default=0.23}
]
